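# Physics asset for the GothClaws medium female backpack cosmetic, giving its pack a dangling rope that can swing.
# Three capsules form the pack body, with a small sphere at base_rope locked to it on all angular axes.
# Five capsule bodies chain from base_rope through dyn_rope_05, each joint limited to moderate swing and slight twist.
# The last link is much larger, suggesting an ornament at the rope's end.
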PhysicsAsset /BRCosmetics/Accessories/FORT_Backpacks/Backpack_F_MED_GothClaws/Meshes/F_MED_GothClaws_Pack_Physics
bodies {
 base_rope [{shapeType=Sphere, radius=1, shapeName=base_rope_sphere}]
 dyn_rope_01 [{shapeType=Capsule, length=1, radius=0.5, shapeName=dyn_rope_01_capsule, center=<0.6796 0 0>, rotation=<rot -90 0 0>}]
 dyn_rope_02 [{shapeType=Capsule, length=1, radius=0.5, shapeName=dyn_rope_02_capsule, center=<0.751 0 0>, rotation=<rot -90 0 0>}]
 dyn_rope_03 [{shapeType=Capsule, length=1, radius=0.7, shapeName=dyn_rope_03_capsule, center=<0.7664 0 0>, rotation=<rot -90 0 0>}]
 dyn_rope_04 [{shapeType=Capsule, length=1, radius=0.7, shapeName=dyn_rope_04_capsule, center=<0.7271 0 0>, rotation=<rot -90 0 0>}]
 dyn_rope_05 [{shapeType=Capsule, length=2, radius=2.5, shapeName=dyn_rope_05_capsule, center=<2.1949 0 0>, rotation=<rot -90 0 0>}]
 pack [
  {shapeType=Capsule, length=19, radius=14, shapeName=pack_capsule, center=<-0.4189 17.3788 -14.009>, rotation=<rot -90 0 0>}
  {shapeType=Capsule, length=8.5, radius=4, shapeName=pack_capsule, center=<0.08 -4.0988 -5.9233>, rotation=<rot -90 -0.0004 0>}
  {shapeType=Capsule, length=15, radius=15, shapeName=pack_capsule, center=<0 11.8188 0>, rotation=<rot 90 0 0>}
 ]
}
constraints [
 {bone1Name=dyn_rope_01, bone2Name=base_rope, swing1LimitDegrees=25, swing2LimitDegrees=25, twistLimitDegrees=5}
 {bone1Name=dyn_rope_02, bone2Name=dyn_rope_01, swing1LimitDegrees=25, swing2LimitDegrees=25, twistLimitDegrees=5}
 {bone1Name=dyn_rope_03, bone2Name=dyn_rope_02, swing1LimitDegrees=25, swing2LimitDegrees=25, twistLimitDegrees=5}
 {bone1Name=dyn_rope_04, bone2Name=dyn_rope_03, swing1LimitDegrees=25, swing2LimitDegrees=25, twistLimitDegrees=5}
 {bone1Name=dyn_rope_05, bone2Name=dyn_rope_04, swing1LimitDegrees=25, swing2LimitDegrees=25, twistLimitDegrees=5}
 {
  bone1Name=base_rope
  bone2Name=pack
  swing1LimitDegrees=45
  swing1Motion=Locked
  swing2LimitDegrees=45
  swing2Motion=Locked
  twistLimitDegrees=45
  twistMotion=Locked
 }
]
