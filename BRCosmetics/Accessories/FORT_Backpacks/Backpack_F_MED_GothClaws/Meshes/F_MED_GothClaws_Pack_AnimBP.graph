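# Anim Blueprint for the GothClaws backpack cosmetic that spins its disk and adds secondary physics motion.
# In the front end, five chained AnimDynamics nodes swing the rope bones; otherwise RigidBody simulates the pack, followed by additive pack bone offsets.
# Each update derives those offsets from the owner's spine_05 rotation relative to the reference pose.
# Both branches share a cached input pose and spin dyn_disk_01 from real time.
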AnimBlueprint /BRCosmetics/Accessories/FORT_Backpacks/Backpack_F_MED_GothClaws/Meshes/F_MED_GothClaws_Pack_AnimBP
target_skeleton=&/F_MED_GothClaws_Pack_Skeleton
graphs [
 AnimGraph {nodes [
   n1 = RigidBody(ComponentPose=Local To Component(LocalPose=Use cached pose 'Base'())) {
    bOverrideWorldGravity=true
    overridePhysicsAsset=&/F_MED_GothClaws_Pack_Physics
    overrideWorldGravity {z=-980}
    simSpaceSettings {dampingAlpha=0, maxAngularAcceleration=500, maxAngularVelocity=6, maxLinearAcceleration=1700, maxLinearVelocity=450, worldAlpha=1}
   }
   n2 = "Transform (Modify) Bone - Bone: pack"(ComponentPose=@n1, Translation="0,0,0", Rotation=Get pickaxe_rotate() -> (pickaxe_rotate:Rotator)) {rotationMode=BMM_Additive, rotationSpace=BCS_BoneSpace, alphaBoolBlend {blendOption=Linear}, boneToModify {boneName=pack}}
   n3 = "Transform (Modify) Bone - Bone: pack"(
    ComponentPose=@n2
    Translation=Get backpack_translation()
    Rotation=Get backpack_rotate()
    Scale=Get backpack_scale() -> (backpack_scale:Vector)
   ) {
    rotationMode=BMM_Additive
    scaleMode=BMM_Additive
    translationMode=BMM_Additive
    alphaBoolBlend {blendOption=Linear}
    boneToModify {boneName=pack}
   }
   n4 = "Transform (Modify) Bone - Bone: dyn_disk_01"!(
    ComponentPose=@n3
    Alpha=1
    Translation="0,0,0"
    Rotation_Roll:float=Get CurrentRotation() -> (CurrentRotation:double)
    Rotation_Pitch:float=0
    Rotation_Yaw:float=0
    Scale="1,1,1"
   ) -> (Pose) {rotationMode=BMM_Additive, rotationSpace=BCS_BoneSpace, boneToModify {boneName=dyn_disk_01}}
   n5 = Use cached pose 'Base'()
   n6 = Anim Dynamics - Bone: dyn_rope_01(
    ComponentPose="Transform (Modify) Bone - Bone: dyn_disk_01"(
     ComponentPose=Local To Component(LocalPose=@n5)
     Alpha=1
     Translation="0,0,0"
     Rotation_Roll=Get CurrentRotation()
     Rotation_Pitch=0
     Rotation_Yaw=0
     Scale="1,1,1"
    ) {rotationMode=BMM_Additive, rotationSpace=BCS_BoneSpace, boneToModify {boneName=dyn_disk_01}}
    Alpha=1
    ExternalForce="0,0,0"
   ) {
    angularDampingOverride=1
    angularSpringConstant=300
    bAngularSpring=true
    bOverrideAngularDamping=true
    bUseGravityOverride=true
    simulationSpace=RootRelative
    boundBone {boneName=dyn_rope_01}
    componentAppliedLinearAccClamp=<3000 3000 3000>
    componentLinearAccScale=<0.4 0.4 0.4>
    componentLinearVelScale=<0.3 0.3 0.3>
    gravityOverride {z=-1100}
    physicsBodyDefinitions [{
      boundBone {boneName=dyn_rope_01}
      boxExtents=<7 7 7>
      constraintSetup {angularLimitsMax=<3 60 50>, angularLimitsMin=<-3 -60 -50>, angularTarget=<1 0 0>}
      localJointOffset=<0.5 0 0>
     }]
   }
   n7 = Anim Dynamics - Bone: dyn_rope_02(ComponentPose=@n6, Alpha=1, ExternalForce="0,0,0") {
    angularDampingOverride=1
    angularSpringConstant=300
    bAngularSpring=true
    bOverrideAngularDamping=true
    bUseGravityOverride=true
    simulationSpace=RootRelative
    boundBone {boneName=dyn_rope_02}
    componentAppliedLinearAccClamp=<3000 3000 3000>
    componentLinearAccScale=<0.4 0.4 0.4>
    componentLinearVelScale=<0.3 0.3 0.3>
    gravityOverride {z=-1100}
    physicsBodyDefinitions [{
      boundBone {boneName=dyn_rope_02}
      boxExtents=<7 7 7>
      constraintSetup {angularLimitsMax=<3 60 50>, angularLimitsMin=<-3 -60 -50>, angularTarget=<1 0 0>}
      localJointOffset=<0.5 0 0>
     }]
   }
   n8 = Anim Dynamics - Bone: dyn_rope_03(ComponentPose=@n7, Alpha=1, ExternalForce="0,0,0") {
    angularDampingOverride=1
    angularSpringConstant=300
    bAngularSpring=true
    bOverrideAngularDamping=true
    bUseGravityOverride=true
    simulationSpace=RootRelative
    boundBone {boneName=dyn_rope_03}
    componentAppliedLinearAccClamp=<3000 3000 3000>
    componentLinearAccScale=<0.4 0.4 0.4>
    componentLinearVelScale=<0.3 0.3 0.3>
    gravityOverride {z=-1100}
    physicsBodyDefinitions [{
      boundBone {boneName=dyn_rope_03}
      boxExtents=<7 7 7>
      constraintSetup {angularLimitsMax=<3 60 50>, angularLimitsMin=<-3 -60 -50>, angularTarget=<1 0 0>}
      localJointOffset=<0.5 0 0>
     }]
   }
   n9 = Anim Dynamics - Bone: dyn_rope_04(ComponentPose=@n8, Alpha=1, ExternalForce="0,0,0") {
    angularDampingOverride=1
    angularSpringConstant=300
    bAngularSpring=true
    bOverrideAngularDamping=true
    bUseGravityOverride=true
    simulationSpace=RootRelative
    boundBone {boneName=dyn_rope_04}
    componentAppliedLinearAccClamp=<3000 3000 3000>
    componentLinearAccScale=<0.4 0.4 0.4>
    componentLinearVelScale=<0.3 0.3 0.3>
    gravityOverride {z=-1100}
    physicsBodyDefinitions [{
      boundBone {boneName=dyn_rope_04}
      boxExtents=<7 7 7>
      constraintSetup {angularLimitsMax=<3 60 50>, angularLimitsMin=<-3 -60 -50>, angularTarget=<1 0 0>}
      localJointOffset=<0.5 0 0>
     }]
   }
   n10 = Anim Dynamics - Bone: dyn_rope_05(ComponentPose=@n9, Alpha=1, ExternalForce="0,0,0") {
    angularDampingOverride=0.9999
    angularSpringConstant=300
    bAngularSpring=true
    bOverrideAngularDamping=true
    bUseGravityOverride=true
    simulationSpace=RootRelative
    boundBone {boneName=dyn_rope_05}
    componentAppliedLinearAccClamp=<3000 3000 3000>
    componentLinearAccScale=<0.4 0.4 0.4>
    componentLinearVelScale=<0.3 0.3 0.3>
    gravityOverride {z=-1100}
    physicsBodyDefinitions [{
      boundBone {boneName=dyn_rope_05}
      boxExtents=<7 7 7>
      constraintSetup {angularLimitsMax=<3 60 50>, angularLimitsMin=<-3 -60 -50>, angularTarget=<1 0 0>}
      localJointOffset=<1 0 0>
     }]
   }
   Output Pose(Result=Blend Poses by bool(BlendPose_0=Component To Local(ComponentPose=@n10), BlendPose_1=Component To Local(ComponentPose=@n4), bActiveValue=Get bIsInFrontEnd())) {layerGroup=None}
   Save cached pose 'Base'(Pose=Input Pose()) {cacheName=Base}
  ]}
 EventGraph {nodes [
   Event BlueprintUpdateAnimation()
   > n1 = Sequence!(execute) -> (then_0, then_1, then_2:Exec)
   n2 = GetOwnerSkeletalMeshComponent()
   n3 = BreakRotator(InRot=BreakTransform(InTransform=GetDeltaTransformFromRefPose(self=@n2, BoneName=spine_05, BaseName=spine_01)).Rotation)
   Set backpack_rotate(execute=@n1.then_1, backpack_rotate=MakeRotator(Roll=MapRangeClamped(Value=@n3.Yaw, InRangeA=0, InRangeB=-60, OutRangeB=50)))
   > Set!(
    execute
    pickaxe_rotate_Roll:float=0
    pickaxe_rotate_Pitch:float=0
    pickaxe_rotate_Yaw:float=MapRangeClamped(Value=@n3, InRangeA=0, InRangeB=20, OutRangeB=-15)
   ) -> (then, Output_Get:Rotator)
   Set!(
    execute=@n1.then_2
    backpack_translation_X:double=MapRangeClamped(Value=@n3.Yaw, InRangeA=0, InRangeB=0, OutRangeB=0)
    backpack_translation_Y:double=MapRangeClamped(Value=@n3.Yaw, InRangeA=-20, InRangeB=-60, OutRangeB=-15)
    backpack_translation_Z:double=MapRangeClamped(Value=@n3.Yaw, InRangeA=0, InRangeB=-60, OutRangeB=-3)
   ) -> (then, Output_Get:Vector)
   n4 = MapRangeClamped(Value=@n3.Yaw, InRangeA=0, InRangeB=-60, OutRangeA=1, OutRangeB=0.9)
   > Set backpack_scale(backpack_scale:Vector=MakeVector(X=@n4, Y=@n4, Z=@n4)) -> (Output_Get:Vector)
   n5 = GetRealTimeSeconds()
   n6 = Get Rotation Speed() -> (Rotation Speed:double)
   Set CurrentRotation(
    execute=@n1.then_0
    CurrentRotation:double="Division (Whole and Remainder)"!(Dividend:double=float * float(A=@n5, B=@n6) {is=title K2Node_PromotableOperator}, Divisor:double=-360) -> (Remainder:double, ReturnValue:int=0)
   ) -> (Output_Get:double)
   TryGetPawnOwner()
  ]}
]
variables [
 backpack_rotate {type=Rotator}
 backpack_scale {type=Vector}
 backpack_translation {type=Vector}
 CurrentRotation {type=number}
 pickaxe_rotate {type=Rotator}
 Rotation Speed {type=number, default=-30}
 Tangly_ExternalForce {type=Vector}
]
